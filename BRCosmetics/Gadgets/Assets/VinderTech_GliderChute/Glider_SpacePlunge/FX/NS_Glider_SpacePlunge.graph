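# Niagara VFX for the SpacePlunge glider cosmetic, drawing thruster flames, mesh exhaust bursts and lightning at the glider's skeletal mesh sockets.
# Thrusters holds two persistent socket-following particles, while CubeThrusters and Hunt_Lightning partition particles across four sockets with per-socket offsets.
# The variant input gates CubeThrusters and Hunt_Lightning and restyles Thrusters.
# A system timeline and ForwardDot modulate intensity, with front-end overrides and time-of-day scaling on glow and color.
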NiagaraSystem /BRCosmetics/Gadgets/Assets/VinderTech_GliderChute/Glider_SpacePlunge/FX/NS_Glider_SpacePlunge
dependencies {
 usedDataInterfaces [
  &NiagaraDataInterfaceVectorCurve
  &NiagaraDataInterfaceColorCurve
  &NiagaraDataInterfaceCurve
  &NiagaraDataInterfaceCamera
  &NiagaraDataInterfaceArrayInt32
  &NiagaraDataInterfaceArrayFloat
  &NiagaraDataInterfaceArrayFloat3
  &NiagaraDataInterfaceMeshRendererInfo
  &NiagaraDataInterfaceSpriteRendererInfo
  &NiagaraDataInterfaceSkeletalMesh
 ]
 usedDynamicInputs [
  &MakeIntFromFloat
  &SetBoolByFloatComparison
  &MakeCustomIntFromBool
  &Boolean_AndOperation
  &Boolean_NotOperation
 ]
 usedModules [
  &self:SystemSpawnScript.SystemScriptSource.SystemScriptGraph.NiagaraNodeAssignment_0.SetVariables_#1
  &self:SystemSpawnScript.SystemScriptSource.SystemScriptGraph.NiagaraNodeAssignment_2.SetVariables_#2
  &self:SystemSpawnScript.SystemScriptSource.SystemScriptGraph.NiagaraNodeAssignment_4.SetVariables_#3
  &self:SystemSpawnScript.SystemScriptSource.SystemScriptGraph.NiagaraNodeAssignment_5.SetVariables_#4
  &self:SystemSpawnScript.SystemScriptSource.SystemScriptGraph.NiagaraNodeAssignment_7.SetVariables_#5
  &SystemState
  &BlendSpaceHelper
  &self:SystemSpawnScript.SystemScriptSource.SystemScriptGraph.NiagaraNodeAssignment_3.SetVariables_#6
  &self:SystemSpawnScript.SystemScriptSource.SystemScriptGraph.NiagaraNodeAssignment_6.SetVariables_#7
  /Game/Effects/Niagara/Modules/Logic/UpdateValueStateChange
  &Timeline
 ]
}
emitters {
 CubeThrusters {
  emitterUpdateScript [
   &EmitterState()
   &SpawnRate(SpawnGroup=1, SpawnRate=50, Use Spawn Probability=false) {enabled=false, module=SpawnRate001}
   &SpawnRate(
    Spawn Probability=&MaskFloat(Mask Source=1, Mask Parameter=$System.VariantIndex, Mask=1, Masked Float=1, Passthrough Float=0)
    SpawnRate=150
    Use Spawn Probability=true
   )
  ]
  particleSpawnScript [
   &PartitionParticles(Number Of Partitions=4)
   &InitializeParticle(
    Color Maximum=<rgba 0.5 0.4869 0.4609 1>
    Color Minimum=<rgba 0.25 0.248 0.2285 1>
    Color Mode=::NewEnumerator2
    Lifetime=&MergeHelper_FloatSwitch(
     0=&RandomRangeFloat(Minimum=0.6, Maximum=0.8)
     1=&RandomRangeFloat(Minimum=0.6, Maximum=0.8)
     2=&RandomRangeFloat(Minimum=1, Maximum=2)
     3=&RandomRangeFloat(Minimum=1, Maximum=2)
     Index=$Particles.PartitionParticles.Partition
     Num=::NewEnumerator2
    )
    Mesh Scale=&MergeHelper_VectorSwitch(
     0=&Multiply_VectorByFloat(Vector=<1 1 1>, Float=1.25)
     1=&Multiply_VectorByFloat(Vector=&RandomRangeVector(Minimum=<1.5 2 1.5>, Maximum=<1.5 4 1.5>), Float=0.6)
     2=<0.5 0.5 0.5>
     3=<0.5 0.5 0.5>
     Index=$Particles.PartitionParticles.Partition
     Num=::NewEnumerator2
    )
    Mesh Scale Mode=::NewEnumerator4
    UsePositionOffset=false
    Write Material Random=true
   )
   &SubUVAnimation(
    Mesh Renderer=DataInterface() {propertyValues {MeshRenderer=&self:CubeThrusters_0.NiagaraMeshRendererProperties_0}}
    Output Alternate Renderer Binding=true
    Random Start Frame=true
    Renderer Type:ENiagaraMeshOrSprite=::NewEnumerator1
    SubUV Lookup Index Scale=0
    UseEndFrame=false
    UseStartFrame=false
   )
   &SkeletalMeshLocation(
    Bone / Socket Index=$Particles.PartitionParticles.Partition
    MaskSpawnGroup=false
    Mesh Sampling Type=::NewEnumerator5
    Sampled Orientation Coordinate Space=::Local
    Sampled Position Coordinate Space=::Local
    Sampled Position Offset=&Multiply_Vector(
     A=&MergeHelper_VectorSwitch(
      0=&Lerp_Vector(
       A=&RandomRangeVector(Minimum=<-15 0 -5>, Maximum=<17 25 5>)
       B=&RandomRangeVector(Minimum=<-10 0 -4>, Maximum=<8 15 4>)
       Alpha=&FN_FrontEndFloatModifier(Gameplay=$System.Timeline.NormalizedTime, Front End=1)
      )
      1=&Lerp_Vector(
       A=&RandomRangeVector(Minimum=<-17 0 -5>, Maximum=<15 25 5>)
       B=&RandomRangeVector(Minimum=<-8 0 -4>, Maximum=<10 15 4>)
       Alpha=&FN_FrontEndFloatModifier(Gameplay=$System.Timeline.NormalizedTime, Front End=1)
      )
      2=&RandomRangeVector(Minimum=<-5 -10 -5>, Maximum=<10 5 5>)
      3=&RandomRangeVector(Minimum=<-15 5 -5>, Maximum=<0 -14 5>)
      Index=$Particles.PartitionParticles.Partition
      Num=::NewEnumerator2
     )
     B=&MergeHelper_VectorSwitch(
      0=&Lerp_Vector(A=<1 1 1>, B=<1 2 1>, Alpha=$System.Timeline.NormalizedTime)
      1=&Lerp_Vector(A=<1 1 1>, B=<1 2 1>, Alpha=$System.Timeline.NormalizedTime)
      2=<1 1 1>
      3=<1 1 1>
      Index=$Particles.PartitionParticles.Partition
      Num=::NewEnumerator2
     )
    )
    Skeletal Mesh=$System.Skeletal Mesh
    Socket Sampling Mode=::NewEnumerator5
   )
   &AddVelocity(Velocity=&FN_FrontEndVector3Modifier(
     Gameplay=&MergeHelper_VectorSwitch(
      0=&Lerp_Vector(A=<0 -25 0>, B=<0 -50 0>, Alpha=$System.Timeline.NormalizedTime)
      1=&Lerp_Vector(A=<0 -25 0>, B=<0 -50 0>, Alpha=$System.Timeline.NormalizedTime)
      2=<-8 10 0>
      3=<8 10 0>
      Index=$Particles.PartitionParticles.Partition
      Num=::NewEnumerator2
     )
     Front End=&MergeHelper_VectorSwitch(0=<0 -75 0>, 1=<0 -75 0>, 2=<-8 10 0>, 3=<8 10 0>, Index=$Particles.PartitionParticles.Partition, Num=::NewEnumerator2)
    ))
   &InitialMeshOrientation(
    Mesh Orientation Mode=::NewEnumerator2
    Rotation=&MergeHelper_VectorSwitch(0=<0 0 0>, 1=<0 0 0>, 2=<0 0 0.125>, 3=<0 0 0.125>, Index=$Particles.PartitionParticles.Partition, Num=::NewEnumerator2)
    Use Rotation Vector=true
   )
  ]
  particleUpdateScript [
   &ParticleState()
   &ScaleMeshSize(Scale Factor=&Multiply_VectorByFloat(
     Vector=&MergeHelper_VectorSwitch(
      0=&Lerp_Vector(A=<1 1 1>, B=<1 1.5 1>, Alpha=$System.Timeline.NormalizedTime)
      1=&Lerp_Vector(A=<1 1 1>, B=<1 1.5 1>, Alpha=$System.Timeline.NormalizedTime)
      2=<1 1 1>
      3=<1 1 1>
      Index=$Particles.PartitionParticles.Partition
      Num=::NewEnumerator2
     )
     Float=&FloatFromCurve(FloatCurve=DataInterface() {propertyValues {Curve {keys [
          <at 0 0 RCIM_Cubic>
          {interpMode=RCIM_Cubic, tangentMode=RCTM_User, time=0.25, value=1}
          {arriveTangent=-3.1185, tangentMode=RCTM_User, time=1, value=0}
         ]}}})
    ))
   &DynamicMaterialParameters(
    Index 0 Param 1=&FloatFromCurve(FloatCurve=DataInterface() {propertyValues {Curve {keys [<at 0 0>,<at 1 4>]}}}, Scale Curve=2)
    Index 0 Param 2=1
    Index 0 Param 3=&FloatFromCurve(
     FloatCurve=DataInterface() {propertyValues {Curve {keys [{interpMode=RCIM_Cubic, tangentMode=RCTM_User, time=0, value=2}, {arriveTangent=-2.5233, tangentMode=RCTM_User, time=1.0059, value=1}]}}}
     Scale Curve=&Waveform(
      "Waveform [1]"=::NewEnumerator9
      "[1] Input Position"=&RandomRangeFloat(Minimum=0, Maximum=1)
      "[1] Frequency"=3
      "[1] Amplitude Min/Max"=<0.75 1>
     )
    )
    Module.Index0Float0Write=true
    Module.Index0Float1Write=true
    Module.Index0Float2Write=true
    Module.Index0Float3Write=true
   )
   &ScaleColor(
    Scale RGB=&Multiply_VectorByFloat(
     Vector=<1 1 1>
     Float=&FloatFromCurve(
      FloatCurve=DataInterface() {propertyValues {Curve {keys [<at 0.1 1.8 RCIM_Cubic>, <at 0.3 2 RCIM_Cubic>, {interpMode=RCIM_Cubic, tangentMode=RCTM_User, time=0.4, value=1.8}, <at 0.9021 3>, <at 1 0>]}}}
      Scale Curve=&Lerp_Float(A=1, B=1.5, Alpha=$System.Timeline.NormalizedTime)
     )
    )
    ScaleA=true
    ScaleRGB=true
   )
   &GravityForce(
    Coordinate Space=::Local
    Gravity=&Multiply_VectorByFloat(
     Vector=<0 -1 0>
     Float=&FloatFromCurve(
      FloatCurve=DataInterface() {propertyValues {Curve {keys [{interpMode=RCIM_Cubic, tangentMode=RCTM_User, time=0.5, value=0}, <at 1 1>]}}}
      Scale Curve=&Multiply_Float(
       A=&Multiply_Float(
        A=&RandomRangeFloat(Minimum=200, Maximum=500)
        B=&MergeHelper_FloatSwitch(0=1, 1=1, 2=-0.25, 3=-0.25, Index=$Particles.PartitionParticles.Partition, Num=::NewEnumerator2)
       )
       B=&Lerp_Float(A=1, B=1.5, Alpha=$System.Timeline.NormalizedTime)
      )
     )
    )
   )
   &UpdateMeshOrientation(
    Rotation Rate=&FloatFromCurve(FloatCurve=DataInterface() {propertyValues {Curve {keys [{interpMode=RCIM_Cubic, tangentMode=RCTM_User, time=0.35, value=0}, <at 1 1>]}}})
    Rotation Vector=&RandomRangeVector(Minimum=<-1 -1 -1>, Maximum=<1 1 1>)
   )
   &SolveForcesAndVelocity(Clamp Velocity=false, Limit Acceleration=false)
   &BlendSpace(
    "Position Blend (World <-> Local)":NiagaraFloat=&MergeHelper_FloatSwitch(
     0=&Lerp_Float(
      A=1
      B=&FloatFromCurve(FloatCurve=DataInterface() {propertyValues {Curve {keys [{interpMode=RCIM_Cubic, tangentMode=RCTM_User, time=0.3, value=1}, <at 1 0.75>]}}})
      Alpha=$System.Timeline.NormalizedTime
     )
     1=&Lerp_Float(
      A=1
      B=&FloatFromCurve(FloatCurve=DataInterface() {propertyValues {Curve {keys [{interpMode=RCIM_Cubic, tangentMode=RCTM_User, time=0.3, value=1}, <at 1 0.75>]}}})
      Alpha=$System.Timeline.NormalizedTime
     )
     2=1
     3=1
     Index=$Particles.PartitionParticles.Partition
     Num=::NewEnumerator2
     Translation Contribution:NiagaraFloat=1
     Rotation Contribution:NiagaraFloat=&MergeHelper_FloatSwitch(0=1, 1=1, 2=1, 3=1, Index=$Particles.PartitionParticles.Partition, Num=::NewEnumerator2)
    )
    Position Coordinate Space=2
   )
   &CameraOffset(Camera Offset Amount=5)
  ]
  rendererClasses [&NiagaraSpriteRendererProperties, &NiagaraMeshRendererProperties]
  renderers [
   {rendererClass=&NiagaraSpriteRendererProperties}
   {rendererClass=&NiagaraMeshRendererProperties, rendererIndex=1}
  ]
 }
 Hunt_Lightning {
  emitterUpdateScript [
   &EmitterState()
   &SpawnRate(
    Spawn Probability=&Multiply_Float(
     A=/Niagara/DynamicInputs/Helpers/DistributionCurveFloat(
      Minimum:NiagaraFloat=0.4
      Maximum:NiagaraFloat=1
      Distribution Curve:NiagaraDataInterfaceCurve=DataInterface() {propertyValues {Curve {
         defaultValue=0
         keys [{interpMode=RCIM_Cubic, tangentMode=RCTM_User, time=0.5, value=0}, {arriveTangent=-0.0819, tangentMode=RCTM_User, time=1, value=1}]
        }}}
      Randomness Mode:ENiagaraRandomnessMode=::NewEnumerator0
      Evaluation Type:ENiagaraRandomnessEvaluation=::NewEnumerator1
      Seed:NiagaraInt32=0
      Fixed Random Seed:NiagaraBool=false
     )
     B=&Lerp_Float(A=0, B=1, Alpha=$User.Variant)
    )
    SpawnRate=&Lerp_Float(
     A=10
     B=150
     Alpha=&FN_FrontEndFloatModifier(Gameplay=&ClampFloat(Float=$User.ForwardDot), Front End=0.65)
    )
    Use Spawn Probability=true
   )
  ]
  particleSpawnScript [
   &PartitionParticles(Number Of Partitions=4)
   &InitializeParticle(
    Color=<rgba 2 0.6769 0 1>
    Lifetime Max=0.75
    Lifetime Min=0.25
    Lifetime Mode=::NewEnumerator1
    Mesh Scale=&Multiply_VectorByFloat(
     Vector=&Lerp_Vector(
      A=&RandomRangeVector(Minimum=<0.5 0.4 0.4>, Maximum=<0.75 0.6 0.6>)
      B=&RandomRangeVector(Minimum=<1 1 1>, Maximum=<1.5 2 2>)
      Alpha=&FloatFromCurve(
       FloatCurve=DataInterface() {propertyValues {Curve {keys [{interpMode=RCIM_Cubic, tangentMode=RCTM_User, time=0.5, value=0}, <at 1 1>]}}}
       CurveIndex=&RandomRangeFloat(Minimum=0, Maximum=1)
      )
     )
     Float=0.65
    )
    Mesh Scale Mode=::NewEnumerator4
    UsePositionOffset=true
    Write Material Random=true
   )
   &CameraOffset(Camera Offset Amount=20) {enabled=false}
   &ShapeLocation(
    MaskSpawnGroup=false
    Override Seed=false
    Sphere Radius=&SelectFloatFromArray(Direct Array Index=$Particles.PartitionParticles.Partition, Float Selection Array=DataInterface() {propertyValues {FloatData [10,10,15,15]}})
    UseOwnerScale=true
   )
   &InitialMeshOrientation(
    Mesh Axis To Orient=<0 1 0>
    Mesh Orientation Mode=::NewEnumerator2
    Orientation Vector=&SelectVectorFromArray(
     Direct Array Index=$Particles.PartitionParticles.Partition
     Vector Selection Array=DataInterface() {propertyValues {FloatData [<1 0 0>, <1 0 0>, <1 0.25 0>, <1 -0.25 0>]}}
    )
    Rotation=&RandomRangeVector(Minimum=<-0.1 -0.1 -0.1>, Maximum=<0.1 0.1 0.1>)
    Use Rotation Vector=true
   ) {module=InitialMeshOrientation002}
   /SetVariables(Particles.Random=&RandomRangeFloat(Minimum=0, Maximum=1)) {isSetParameters=true}
  ]
  particleUpdateScript [
   &SkeletalMeshLocation(
    Bone / Socket Index=&SelectIntFromArray(Direct Array Index=$Particles.PartitionParticles.Partition, Int Selection Array=DataInterface() {propertyValues {IntData [0,1,2,3]}})
    MaskSpawnGroup=false
    Mesh Sampling Type=::NewEnumerator5
    Sampled Position Offset=&MergeHelper_VectorSwitch(
     0=<0 0 0>
     1=<0 0 0>
     2=&RandomRangeVector(Minimum=<0 0 0>, Maximum=<0 100 0>)
     3=&RandomRangeVector(Minimum=<0 0 0>, Maximum=<0 100 0>)
     Index=$Particles.PartitionParticles.Partition
     Num=::NewEnumerator2
    )
    Skeletal Mesh=$System.Skeletal Mesh
    Socket Sampling Mode=::NewEnumerator5
   )
   &ParticleState()
   &ScaleMeshSize()
   &ScaleColor(
    Scale Alpha=&Multiply_Float(
     A=10
     B=&FloatFromCurve(FloatCurve=DataInterface() {propertyValues {Curve {keys [
          {interpMode=RCIM_Cubic, tangentMode=RCTM_User, time=0, value=0}
          {interpMode=RCIM_Cubic, tangentMode=RCTM_User, time=0.1, value=0}
          {interpMode=RCIM_Cubic, tangentMode=RCTM_User, time=0.2, value=0.8}
          {interpMode=RCIM_Cubic, tangentMode=RCTM_User, time=1, value=0}
         ]}}})
    )
    Scale RGB=&Multiply_VectorByFloat(
     Vector=&Multiply_VectorByFloat(Vector=<1 1 1>, Float=&FN_FrontEndFloatModifier(Gameplay=1, Front End=0.5))
     Float=&Multiply_Float(
      A=&RandomRangeFloat(Minimum=0.5, Maximum=1, Evaluation Type=::NewEnumerator1)
      B=&Multiply_Float(A=&FNTimeOfDayScalar(Night Scale=0.2, Morning Scale=1, Evening Scale=1), B=10)
     )
    )
    ScaleA=true
    ScaleRGB=true
   )
   &DynamicMaterialParameters(
    Index 0 Param 1=&FloatFromCurve(FloatCurve=DataInterface() {propertyValues {Curve {keys [{tangentMode=RCTM_User, time=0, value=1}, {tangentMode=RCTM_User, time=1, value=1}]}}})
    Index 0 Param 2=0.7
    Index 0 Param 3=&FloatFromCurve(FloatCurve=DataInterface() {propertyValues {Curve {keys [
         {arriveTangent=4.0072, interpMode=RCIM_Cubic, leaveTangent=4.0072, tangentMode=RCTM_User, time=0, value=0}
         {arriveTangent=2.7083, interpMode=RCIM_Cubic, leaveTangent=2.7083, time=0.2543, value=0.6225}
         {arriveTangent=1.8473, interpMode=RCIM_Cubic, leaveTangent=1.8473, time=0.6, value=1.625}
         {arriveTangent=1.6993, interpMode=RCIM_Cubic, leaveTangent=1.6993, tangentMode=RCTM_User, time=1, value=2}
        ]}}})
    Index 0 Param 4=5
    Index 1 Param 1=&FN_FrontEndFloatModifier(Gameplay=0.75, Front End=0.65)
    Index 1 Param 2=$Particles.Random
    Index 1 Param 3=1.3
    Index 1 Param 4=1
    Index 2 Param 1=&FloatFromCurve(FloatCurve=DataInterface() {propertyValues {Curve {keys [<at 0 1>,<at 1 0.3>]}}})
    Module.Index0Float0Write=true
    Module.Index0Float1Write=true
    Module.Index0Float2Write=true
    Module.Index0Float3Write=true
    Module.Index1Float0Write=true
    Module.Index1Float1Write=true
    Module.Index1Float2Write=true
    Module.Index1Float3Write=true
    Module.Index2Float0Write=true
    Module.Index2Float1Write=true
    Module.Index2Float2Write=true
    Module.Index2Float3Write=true
    Write Parameter Index 1=::NewEnumerator1
    Write Parameter Index 2=::NewEnumerator1
   )
   &SolveForcesAndVelocity(Clamp Velocity=false, Limit Acceleration=false)
   /SetVariables(Particles.Lightning glow color:LinearColor=&MakeLinearColorFromVectorAndFloat(
     "Vector (RGB)"=&MakeVectorFromLinearColorRGB(Color=&MultiplyLinearColor(
       Linear Color=<rgba 4 1.2362 0 1>
       Scale Factor=&MakeVector4_FromFloat(Float=&RandomRangeFloat(Minimum=0.5, Maximum=1, Evaluation Type=::NewEnumerator1))
      ))
     "Float (Alpha)"=&Multiply_Float(
      A=&FloatFromCurve(FloatCurve=DataInterface() {propertyValues {Curve {keys [<at 0 1 RCIM_Cubic>, {interpMode=RCIM_Cubic, leaveTangent=-3.6307, time=0.9, value=0}]}}})
      B=1
     )
    )) {isSetParameters=true}
   /SetVariables(Particles.Glow Dy02:Vector4f=&MakeVector4(
     X=&FloatFromCurve(FloatCurve=DataInterface() {propertyValues {Curve {keys [{tangentMode=RCTM_User, time=0, value=0.2}, {tangentMode=RCTM_User, time=0.5, value=1}]}}})
     Y=0.7
     Z=0
     W=5
    )) {isSetParameters=true}
   /SetVariables(Particles.Glow Dy03:Vector4f=&MakeVector4(
     X=&FloatFromCurve(
      FloatCurve=DataInterface() {propertyValues {Curve {keys [<at 0 1>,<at 1 0.5>]}}}
      Scale Curve=&FN_FrontEndFloatModifier(Gameplay=2, Front End=1.2)
     )
     Y=$Particles.Random
     Z=1.3
     W=1
    )) {isSetParameters=true}
  ]
  rendererClasses [&NiagaraMeshRendererProperties]
  renderers [
   {rendererClass=&NiagaraMeshRendererProperties, rendererIndex=0}
   {rendererClass=&NiagaraMeshRendererProperties, rendererIndex=1}
  ]
 }
 Thrusters {
  emitterUpdateScript [
   &EmitterState()
   &SpawnBurst_Instantaneous(Spawn Count=2, Use Spawn Probability=false, UseLoopCountLimit=true)
  ]
  particleSpawnScript [
   &PartitionParticles()
   &InitializeParticle(
    Color=<rgba 1 0.4802 0.1013 1>
    Mesh Scale=&MaskVector(
     Mask Source=1
     Mask Parameter=$System.VariantIndex
     Mask=1
     Masked Vector=&FN_FrontEndVector3Modifier(Gameplay=<1.25 0.5 0.5>, Front End=<0.85 0.5 0.5>)
     Passthrough Vector=<1 1 1>
    )
    Mesh Scale Mode=::NewEnumerator4
    Sprite Size=&Multiply_Vector2D_ByFloat(
     Float=&FN_FrontEndFloatModifier(Gameplay=1, Front End=0.75)
     Vector2D=/Niagara/DynamicInputs/Helpers/MaskVector2d(
      Mask Source:NiagaraInt32=1
      Mask Parameter:NiagaraInt32=$System.VariantIndex
      Mask:NiagaraInt32=1
      Masked Vector2D:Vector2f=<80 30>
      Passthrough Vector2D:Vector2f=<40 15>
     )
    )
    Sprite Size Mode=::NewEnumerator4
    UsePositionOffset=true
    Write Material Random=true
   )
   &InitialMeshOrientation(Mesh Orientation Mode=::NewEnumerator2, Orientation Vector=<0 -1 0>, Use Rotation Vector=true)
   /SetVariables(
    Particles.TrailColor:LinearColor=<rgba 1 0.467 0.088 1>
    Particles.TrailOrient:Quat4f=&OrientMeshToVector(Look At Direction=<0 -1 0>, Reference Axis Vector=<0 0 1>)
   ) {isSetParameters=true}
  ]
  particleUpdateScript [
   &ParticleState(Kill Particles When Lifetime Has Elapsed=false, Let Infinitely Lived Particles Die When Emitter Deactivates=true)
   &UpdateMeshOrientation(Rotation Rate=100, Rotation Vector=<0 1 0>) {enabled=false}
   &ScaleMeshSize(Scale Factor=&VectorFromCurve(
     VectorCurve=DataInterface() {propertyValues {XCurve {keys [<at 0 0.5>,<at 1 1>]}, YCurve {keys [<at 0 0.5>,<at 1 1>]}, ZCurve {keys [<at 0 0.25>,<at 1 1>]}}}
     CurveIndex=$Emitter.Age
    )) {module=ScaleMeshSize001}
   &ScaleMeshSize(Scale Factor=&FN_FrontEndVector3Modifier(
     Gameplay=&Lerp_Vector(
      A=&Lerp_Vector(
       A=<0.65 1 1>
       B=<1 1 1>
       Alpha=&RemapRange(
        Input Value To Remap=&ClampFloat(Float=$User.ForwardDot, Min=-1, Max=0)
        Input Min=-1
        Input Max=0
        Output Min=0
        Output Max=1
       )
      )
      B=&MaskVector(Mask Source=1, Mask Parameter=$System.VariantIndex, Mask=1, Masked Vector=<1.5 2 2>, Passthrough Vector=<3 1 1>)
      Alpha=$System.Timeline.NormalizedTime
     )
     Front End=&VectorFromCurve(
      VectorCurve=DataInterface() {propertyValues {XCurve {keys [<at 0 0.25>,<at 1 1>]}, YCurve {keys [<at 0 1>,<at 1 1>]}, ZCurve {keys [<at 0 1>,<at 1 1>]}}}
      Scale Curve=<1.5 1 1>
     )
    ))
   &Color(Color=&MaskLinearColor(
     Mask Source=1
     Mask Parameter=$System.VariantIndex
     Mask=1
     Masked Linear Color=<rgba 1 1 1 1>
     Passthrough Linear Color=&LerpLinearColors(
      StartColor=<rgba 1 0.9287 0.8646 0.9>
      EndColor=&ColorFromCurve(
       ColorCurve=DataInterface() {propertyValues {
         AlphaCurve {keys [<at 0 0.5 RCIM_Cubic>,<at 1 0.5 RCIM_Cubic>]}
         BlueCurve {keys [<at 0 0.035 RCIM_Cubic>,<at 1 0 RCIM_Cubic>]}
         GreenCurve {keys [<at 0 0.8793 RCIM_Cubic>,<at 1 0.6398 RCIM_Cubic>]}
         RedCurve {keys [<at 0 1 RCIM_Cubic>,<at 1 0.75 RCIM_Cubic>]}
        }}
       CurveIndex=&FNTimeOfDayScalar(Night Scale=1, Morning Scale=0, Day Scale=0, Evening Scale=1)
      )
      LerpFactor=&FloatFromCurve(
       FloatCurve=DataInterface() {propertyValues {Curve {keys [{interpMode=RCIM_Cubic, tangentMode=RCTM_User, time=0.6, value=0}, <at 1 1>]}}}
       CurveIndex=&Abs_Float(Float:NiagaraFloat=&DotProduct(Vector A=<0 0 1>, Vector B=&ReturnCameraVector(Camera Vectors=::NewEnumerator1, Coordinate Space=::World)))
      )
     )
    ))
   /SetVariables(
    Particles.OuterThrustScale:Vector3f=&Multiply_Vector(A=$Particles.Scale, B=<0.12 0.15 0.05>)
    Particles.SpriteFacing=<0 1 0>
    Particles.ThrusterCross_Scale:Vector3f=&Multiply_Vector(
     A=$Particles.Scale
     B=&FN_FrontEndVector3Modifier(Gameplay=<1.25 0.75 0.25>, Front End=<1.25 0.75 0.75>)
    )
    Particles.ThrusterScale_Hunt:Vector3f=&Multiply_Vector(A=$Particles.Scale, B=<1.5 2 2>)
    Particles.TrailOrient:Quat4f=&MakeQuaternion(
     Quaternion From=0
     Axis=<0 1 0>
     Angle=&Lerp_Float(A=0, B=-360, Alpha=$Particles.NormalizedAge)
     From=?
     To=?
     Angle Type=::NewEnumerator0
     Normalize Vectors=false
    )
    Particles.TrailScale:Vector3f=<0.5 0.5 1>
    Particles.VisibilityTag=&MergeHelper_IntSwitch(0=0, 1=1, 2=?, 3=?, 4=?, 5=?, Index=$System.VariantIndex, Num=::NewEnumerator0)
   ) {isSetParameters=true}
   &SkeletalMeshLocation(
    Bone / Socket Index=&SelectIntFromArray(Direct Array Index=$Particles.PartitionParticles.Partition, Int Selection Array=DataInterface() {propertyValues {IntData [0,1]}})
    MaskSpawnGroup=false
    Mesh Sampling Type=::NewEnumerator5
    Sampled Position Offset=&FN_FrontEndVector3Modifier(
     Gameplay=&MaskVector(Mask Source=1, Mask Parameter=$System.VariantIndex, Mask=0, Masked Vector=<0 5 0>, Passthrough Vector=<0 4 0>)
     Front End=<0 3 0>
    )
    Skeletal Mesh=$System.Skeletal Mesh
    Socket Sampling Mode=::NewEnumerator5
   )
   &CameraOffset(Camera Offset Amount=&FN_FrontEndFloatModifier(
     Gameplay=&MaskFloat(Mask Source=1, Mask Parameter=$System.VariantIndex, Mask=1, Masked Float=20, Passthrough Float=2)
     Front End=0
    ))
  ]
  rendererClasses [&NiagaraSpriteRendererProperties, &NiagaraMeshRendererProperties]
  renderers [
   {rendererClass=&NiagaraSpriteRendererProperties}
   {rendererClass=&NiagaraMeshRendererProperties, rendererIndex=1}
   {rendererClass=&NiagaraMeshRendererProperties, rendererIndex=2}
   {rendererClass=&NiagaraMeshRendererProperties, rendererIndex=3}
   {rendererClass=&NiagaraSpriteRendererProperties, rendererIndex=4}
   {rendererClass=&NiagaraMeshRendererProperties, rendererIndex=5}
  ]
 }
}
summary {
 systemName=NS_Glider_SpacePlunge
 emitters [
  {emitterName=Thrusters, rendererClasses [&NiagaraSpriteRendererProperties, &NiagaraMeshRendererProperties]}
  {emitterName=Hunt_Lightning, rendererClasses [&NiagaraMeshRendererProperties]}
  {emitterName=CubeThrusters, rendererClasses [&NiagaraSpriteRendererProperties, &NiagaraMeshRendererProperties]}
 ]
 userVariables [
  User.bIsFrontEnd {type=NiagaraBool}
  User.bIsFrontEndPreview {type=NiagaraBool}
  User.bIsFullyDeployed {type=NiagaraBool}
  User.ForwardDot {type=NiagaraFloat, defaultValue=0, description=""}
  User.RightDot {type=NiagaraFloat, defaultValue=0, description=""}
  User.Variant {type=NiagaraFloat, defaultValue=0, description=""}
 ]
}
